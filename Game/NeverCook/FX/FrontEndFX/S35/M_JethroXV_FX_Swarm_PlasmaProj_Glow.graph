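# Additive unlit, two-sided glow material for Niagara mesh particles, apparently for a swarm plasma projectile in the S35 front-end FX.
# A ramp texture, vertex-blue-boosted color, rotatable UV gradient exposure and particle color drive emissive.
# Opacity uses near-camera and depth fades, and WPO adds camera offset plus an Alpha-driven collapse toward object center.
# MF_Jethro_StormMask wraps the output; a timed BulletOrigin offset branch is left unconnected.
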Material /Game/NeverCook/FX/FrontEndFX/S35/M_JethroXV_FX_Swarm_PlasmaProj_Glow
graph {
 properties {
  bUseMaterialAttributes=true
  bUsedWithNiagaraMeshParticles=true
  blendMode=BLEND_Additive
  pixelDepthOffsetMode=PDOM_Legacy
  shadingModel=MSM_Unlit
  twoSided=true
  usedShadingModels=Unlit
 }
 nodes [
  n1 = TextureSample() {texture=/Game/Global/Textures/Default/Masks_and_Gradients/T_Ramp_Virt}
  n2 = VectorParameter() {parameterName=Color, defaultValue=<rgba 1 1 1 1>}
  n3 = LinearInterpolate(A=Multiply(A=@n2.RGB) {constB=2}, B=@n2.RGB, Alpha=VertexColor().B)
  n4 = TextureCoordinate()
  n5 = ScalarParameter() {parameterName=Rotation_Angle}
  n6 = OneMinus(None=StaticComponentMaskParameter(None=MaterialFunctionCall(UVs=@n4, "Rotation Angle (0-1)"=@n5) {materialFunction=&CustomRotator}.Rotated Values) {defaultG=true, parameterName=MaskChannel})
  n7 = Power(Base=@n6, Exp=ScalarParameter() {defaultValue=5, parameterName=Squeez_Top_Power}) {constExponent=5}
  n8 = LinearInterpolate(B=ScalarParameter() {defaultValue=1.5, parameterName=Exposure}, Alpha=Multiply(A=@n6, B=OneMinus(None=@n7))) {constB=1.5}
  n9 = ParticleColor()
  n10 = Multiply(A=Multiply(A=Multiply(A=@n1.R, B=@n3), B=@n8), B=@n9.RGB)
  n11 = ScalarParameter() {defaultValue=25, parameterName=DepthFadeDist}
  n12 = MaterialFunctionCall(Distance=ScalarParameter() {defaultValue=256, parameterName=Near Camera Fade}) {materialFunction=&Near_Camera_Fade}
  n13 = Multiply(
   A=ShadingPathSwitch(Default=DepthFade(Opacity=@n12.Result, FadeDistance=@n11), Mobile=@n12.Result) {desc=Migrated from FeatureLevelSwitch}
   B=@n8
  )
  n14 = ObjectPositionWS()
  n15 = WorldPosition()
  n16 = ScalarParameter() {defaultValue=1, parameterName=Alpha}
  n17 = Add(
   A=MaterialFunctionCall(Offset Amount=ScalarParameter() {defaultValue=5, parameterName=Camera Offset Amount}) {materialFunction=/Engine/Functions/Engine_MaterialFunctions02/WorldPositionOffset/CameraOffset}.Camera Offset
   B=Multiply(A=Subtract(A=@n14, B=@n15.XYZ), B=OneMinus(None=@n16))
  )
  MaterialFunctionCall!(In=MakeMaterialAttributes!(
    BaseColor
    Metallic
    Specular
    Roughness
    Anisotropy
    EmissiveColor=MaterialFunctionCall(Emissive=@n10) {materialFunction=&MF_FX_EmissiveAdjustments}.Result
    Opacity=Multiply(A=@n13, B=@n9.A)
    OpacityMask
    Normal
    Tangent
    WorldPositionOffset=Add(A=@n17) {constB=0}
    SubsurfaceColor
    ClearCoat
    ClearCoatRoughness
    AmbientOcclusion
    Refraction
    CustomizedUVs_0
    CustomizedUVs_1
    CustomizedUVs_2
    CustomizedUVs_3
    CustomizedUVs_4
    CustomizedUVs_5
    CustomizedUVs_6
    CustomizedUVs_7
    PixelDepthOffset
    ShadingModel
    Displacement
   )) {materialFunction=/Game/Effects/Fort_Effects/FrontEnd/S35/MF_Jethro_StormMask}
  n18 = Subtract(A=Time() {period=5}, B=ScalarParameter() {parameterName=StartTime})
  n19 = ScalarParameter() {defaultValue=0.1, group=Timestamp, parameterName=WPODuration}
  n20 = MaterialFunctionCall!(In=Clamp(None=Divide(A=@n18, B=@n19))) {materialFunction=/Game/Effects/Fort_Effects/Materials/_Functions/MF_Smoothstep_EaseOut}
  n21 = VectorParameter() {parameterName=BulletOrigin}
  n22 = ActorPositionWS()
  Multiply(A=Multiply(A=Clamp(None=@n20.Result), B=Subtract(A=@n21.RGB, B=@n22)), B=VertexColor().R)
 ]
}
